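; Parent gameplay ability for Athena item powers, supplying shared setup and gameplay cue helpers.
; InitAbility caches the avatar as a Fort player pawn, printing an error on failure, then fires the activation cue with parameters from the stub ActivationGCParams.
; Burst and looping cue helpers check for a valid tag and optionally pass parameters; the burst helper also forces a pawn net update.
; GetRowAtLevel evaluates scalable floats at ability level; PowerCommitFailed is an empty event.
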(event Custom|PowerCommitFailed)

(fn ActivationGCParams
  (return 0))

(fn AddGCLoop
  (params (GameplayCueTag :type Gameplay%20Tag%20Structure) (WithParams :type Boolean) (GameplayCueParameters :type Gameplay%20Cue%20Parameters%20Structure%20%28by%20ref%29) (bRemoveOnAbilityEnd :type Boolean))
  ;[1 1 1] Play A GC if it is a valid tag. Use params version of the execution if "WithParams" is true.
  (if (GameplayTags|NotEqual(GameplayTag) :A GameplayCueTag)
    (if WithParams
      (Ability|AddGameplayCueWithParamsToOwner GameplayCueTag GameplayCueParameters bRemoveOnAbilityEnd)
      (else
        (Ability|AddGameplayCueToOwner :GameplayCueTag GameplayCueTag :bRemoveOnAbilityEnd bRemoveOnAbilityEnd)))))

(fn GetRowAtLevel
  (params (Row :type Scalable%20Float%20Structure%20%28by%20ref%29))
  ;[1 1 1] Get Row at the Ability Level
  (return (ScalableFloat|GetValueatLevel Row (Math|Conversions|ToFloat(Integer) (Ability|GetAbilityLevel)))))

(fn InitAbility
  (bind _returnvalue (Ability|GetAvatarActorfromActorInfo))
  ;[1 1 1] Save Out Player Pawn
  (Utilities|IsValid _returnvalue
    (:Is%20Valid
      (bind _asfort_player_pawn (Utilities|Casting|CastToFortPlayerPawn _returnvalue)
        (:then
          (Variables|Default|SetPlayerPawn _asfort_player_pawn)
          ;[1 1 1] Play Activation GC
          (CallFunction|PlayBurstGC (Variables|PowerGameplayCues|GetGC_Activation) (Variables|PowerGameplayCues|GetActivationGCUsesParams) (CallFunction|ActivationGCParams)))
        (:CastFailed
          (Development|PrintString "Power Init: Avatar is not a PlayeraPawn"))))
    (:Is%20Not%20Valid
      (Development|PrintString "Power Init: Avatar is not valid"))))

(fn PlayBurstGC
  (params (GameplayCueTag :type Gameplay%20Tag%20Structure) (WithParams :type Boolean) (GameplayCueParameters :type Gameplay%20Cue%20Parameters%20Structure%20%28by%20ref%29))
  ;[1 1 1] Play A GC if it is a valid tag. Use params version of the execution if "WithParams" is true.
  (if (GameplayTags|NotEqual(GameplayTag) :A GameplayCueTag)
    (if WithParams
      (Ability|ExecuteGameplayCueWithParamsOnOwner GameplayCueTag GameplayCueParameters)
      (else
        (Ability|ExecuteGameplayCueOnOwner GameplayCueTag)))
    (Networking|ForceNetUpdate (Variables|Default|GetPlayerPawn))))
